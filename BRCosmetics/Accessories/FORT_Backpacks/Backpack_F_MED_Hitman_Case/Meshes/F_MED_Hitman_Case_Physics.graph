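# Physics asset for the Hitman Case backpack cosmetic, giving its knife and four straps physical bodies for secondary motion.
# Each dynamic bone (knife, top straps A/B, bottom straps A/B) has a single capsule body, and the pack bone has a small capsule as the anchor.
# Five constraints tie each dangling piece directly to the pack in a star layout, with narrow swing limits and very tight twist to keep movement subtle.
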PhysicsAsset /BRCosmetics/Accessories/FORT_Backpacks/Backpack_F_MED_Hitman_Case/Meshes/F_MED_Hitman_Case_Physics
bodies {
 dyn_bot_strapA [{shapeType=Capsule, length=3.25, radius=2.5, center=<-0.6028 0 -0.4877>, rotation=<rot 51.0268 180 180>}]
 dyn_bot_strapB [{shapeType=Capsule, length=3.25, radius=3.25, center=<-0.6633 0 -0.3382>, rotation=<rot 62.9865 -180 -180>}]
 dyn_knife [{shapeType=Capsule, length=4.5, radius=5, center=<-0.3703 0 -0.9586>, rotation=<rot -41.0928 0 0>}]
 dyn_top_strapA [{shapeType=Capsule, length=4, radius=2.75, center=<0.5 0 0.6037>, rotation=<rot 34.2338 180 180>}]
 dyn_top_strapB [{shapeType=Capsule, length=4.75, radius=3, center=<-0.9006 -0.0001 -2.2361>, rotation=<rot 21.9379 -180 -180>}]
 pack [{shapeType=Capsule, length=1, radius=1}]
}
constraints [
 {bone1Name=dyn_knife, bone2Name=pack, swing1LimitDegrees=15, swing2LimitDegrees=15, twistLimitDegrees=5}
 {bone1Name=dyn_top_strapA, bone2Name=pack, swing1LimitDegrees=15, swing2LimitDegrees=23, twistLimitDegrees=3}
 {bone1Name=dyn_top_strapB, bone2Name=pack, swing1LimitDegrees=15, swing2LimitDegrees=27, twistLimitDegrees=3}
 {bone1Name=dyn_bot_strapA, bone2Name=pack, swing1LimitDegrees=15, swing2LimitDegrees=23.5714, twistLimitDegrees=3}
 {bone1Name=dyn_bot_strapB, bone2Name=pack, swing1LimitDegrees=15, swing2LimitDegrees=18.2857, twistLimitDegrees=3}
]
